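# Niagara system for the S38 frontend that blows green to yellow-green leaf sprites across the scene from an offset box.
# The active Leaves001 emitter spawns in bursts via a loop-age curve with 50% spawn probability.
# Wind, gravity and aerodynamic drag move the spinning, mesh-orientation-aligned sprites, which fade in and shrink late in life.
# The Leaves emitter is a disabled variant without wind or active gravity.
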NiagaraSystem /Game/Effects/Fort_Effects/FrontEnd/S38/NS_S3800_Frontend
dependencies {usedDataInterfaces [&NiagaraDataInterfaceCurve], usedDynamicInputs [], usedModules [&SystemState]}
emitters {
 Leaves {
  bEnabled=false
  emitterUpdateScript [
   &EmitterState()
   &SpawnRate(SpawnRate=25, Use Spawn Probability=false)
   &SpawnRate(
    Spawn Probability=0.5
    SpawnRate=&FloatFromCurve(
     FloatCurve=DataInterface() {propertyValues {Curve {
        defaultValue=1
        keys [
         {interpMode=RCIM_Cubic, tangentMode=RCTM_User, time=0, value=0}
         {interpMode=RCIM_Cubic, tangentMode=RCTM_User, time=0.1, value=1.025}
         {arriveTangent=-1.875, interpMode=RCIM_Cubic, leaveTangent=-1.875, time=0.2, value=0}
         {arriveTangent=0.25, interpMode=RCIM_Cubic, leaveTangent=0.25, time=0.3, value=0.65}
         {arriveTangent=-0.9286, interpMode=RCIM_Cubic, leaveTangent=-0.9286, time=0.4, value=0.05}
         {tangentMode=RCTM_User, time=1, value=0}
        ]
       }}}
     CurveIndex=$Emitter.NormalizedLoopAge
     Scale Curve=5
    )
    Use Spawn Probability=true
   ) {module=SpawnRate001}
  ]
  particleSpawnScript [
   &InitializeParticle(
    Color Maximum=<rgba 0.3011 0.3247 0.0242 1>
    Color Minimum=<rgba 0.0281 0.2101 0.055 1>
    Color Mode=::NewEnumerator2
    Lifetime Max=14
    Lifetime Min=10
    Lifetime Mode=::NewEnumerator1
    Sprite Rotation Mode=::NewEnumerator2
    Sprite Size Mode=::NewEnumerator1
    Uniform Sprite Size Max=20
    Uniform Sprite Size Min=10
    UsePositionOffset=false
    Write Material Random=true
   )
   &ShapeLocation(
    Box Size=<1 100 400>
    MaskSpawnGroup=false
    Offset=<0 700 200>
    Offset Mode=::NewEnumerator0
    Override Seed=false
    Shape Primitive=::NewEnumerator2
    Surface Only=false
    UseOwnerScale=true
   )
   &GravityForce(Gravity=<0 0 -30>) {enabled=false}
   &CurlNoiseForce(Bias Noise=false, Noise Frequency=25, Noise Strength=50, Pan Noise=true, Pan Noise Field=<2 1 0.5>) {enabled=false}
   &ApplyInitialForces()
   &AddVelocity(Velocity=<0 1 0>, Velocity Speed Scale=&RandomRangeFloat(Minimum=-50, Maximum=-100)) {enabled=false}
   &DynamicMaterialParameters(
    Index 0 Param 1=0.1
    Module.Index0Float0Write=true
    Module.Index0Float1Write=true
    Module.Index0Float2Write=true
    Module.Index0Float3Write=true
   )
  ]
  particleUpdateScript [
   &ParticleState()
   &SpriteRotationRate(Rotation Rate=&RandomRangeFloat(Minimum=75, Maximum=150))
   &ScaleSpriteSize(Uniform Curve Sprite Scale=DataInterface() {propertyValues {Curve {
       defaultValue=0
       keys [{interpMode=RCIM_Cubic, tangentMode=RCTM_User, time=0, value=1}, {interpMode=RCIM_Cubic, tangentMode=RCTM_User, time=1, value=0}]
      }}})
   &Drag(Use Linear Drag=true, Use Rotational Drag=false) {enabled=false}
   &SolveForcesAndVelocity(Clamp Velocity=true, Limit Acceleration=false, Speed Limit=800)
   &ScaleColor(
    Scale Alpha=&FloatFromCurve(FloatCurve=DataInterface() {propertyValues {Curve {
        defaultValue=1
        keys [
         {arriveTangent=10.5294, interpMode=RCIM_Cubic, leaveTangent=10.5294, tangentMode=RCTM_User, time=0, value=0}
         {interpMode=RCIM_Cubic, tangentMode=RCTM_User, time=0.1, value=1}
        ]
       }}})
    ScaleA=true
    ScaleRGB=false
   )
  ]
  rendererClasses [&NiagaraSpriteRendererProperties]
  renderers [{rendererClass=&NiagaraSpriteRendererProperties}]
 }
 Leaves001 {
  emitterUpdateScript [
   &EmitterState()
   &SpawnRate(SpawnRate=25, Use Spawn Probability=false) {enabled=false}
   &SpawnRate(
    Spawn Probability=0.5
    SpawnRate=&FloatFromCurve(
     FloatCurve=DataInterface() {propertyValues {Curve {
        defaultValue=1
        keys [
         {interpMode=RCIM_Cubic, tangentMode=RCTM_User, time=0, value=0}
         {interpMode=RCIM_Cubic, tangentMode=RCTM_User, time=0.1, value=1.025}
         {arriveTangent=-1.875, interpMode=RCIM_Cubic, leaveTangent=-1.875, time=0.2, value=0}
         {arriveTangent=0.25, interpMode=RCIM_Cubic, leaveTangent=0.25, time=0.3, value=0.65}
         {arriveTangent=-0.9286, interpMode=RCIM_Cubic, leaveTangent=-0.9286, time=0.4, value=0.05}
         {tangentMode=RCTM_User, time=1, value=0}
        ]
       }}}
     CurveIndex=$Emitter.NormalizedLoopAge
     Scale Curve=5
    )
    Use Spawn Probability=true
   ) {module=SpawnRate001}
  ]
  particleSpawnScript [
   &InitializeParticle(
    Color Maximum=<rgba 0.3011 0.3247 0.0242 1>
    Color Minimum=<rgba 0.0281 0.2101 0.055 1>
    Color Mode=::NewEnumerator2
    Lifetime Max=20
    Lifetime Min=15
    Lifetime Mode=::NewEnumerator1
    Sprite Rotation Mode=::NewEnumerator2
    Sprite Size Mode=::NewEnumerator1
    Uniform Sprite Size Max=20
    Uniform Sprite Size Min=10
    UsePositionOffset=false
    Write Material Random=true
   )
   &ShapeLocation(
    Box Size=<1 100 400>
    MaskSpawnGroup=false
    Offset=<0 700 200>
    Offset Mode=::NewEnumerator0
    Override Seed=false
    Shape Primitive=::NewEnumerator2
    Surface Only=false
    UseOwnerScale=true
   )
   &CurlNoiseForce(Bias Noise=false, Noise Frequency=25, Noise Strength=50, Pan Noise=true, Pan Noise Field=<2 1 0.5>) {enabled=false}
   &ApplyInitialForces()
   &AddVelocity(Velocity=<0 1 0>, Velocity Speed Scale=&RandomRangeFloat(Minimum=-50, Maximum=-100)) {enabled=false}
   &DynamicMaterialParameters(
    Index 0 Param 1=0.1
    Module.Index0Float0Write=true
    Module.Index0Float1Write=true
    Module.Index0Float2Write=true
    Module.Index0Float3Write=true
   )
  ]
  particleUpdateScript [
   &ParticleState()
   &SpriteRotationRate(Rotation Rate=&RandomRangeFloat(Minimum=75, Maximum=150))
   &ScaleSpriteSize(Uniform Curve Sprite Scale=DataInterface() {propertyValues {Curve {
       defaultValue=0
       keys [{interpMode=RCIM_Cubic, tangentMode=RCTM_User, time=0.7, value=1}, {interpMode=RCIM_Cubic, tangentMode=RCTM_User, time=1, value=0}]
      }}})
   &WindForce(Frequency=2, Scale=0.7, Wind Speed=<0 -1 0>, Wind Speed Scale=175)
   &Drag(Use Linear Drag=true, Use Rotational Drag=false) {enabled=false}
   &GravityForce(Gravity=<0 0 -30>)
   &AerodynamicDrag(Aerodynamic Drag=0.5, Aerodynamic Rotational Drag=0.5, Use Linear Drag=true, Use Rotational Drag=true)
   &SolveForcesAndVelocity(Clamp Velocity=true, Limit Acceleration=false, Speed Limit=800)
   &AlignSpriteToMeshOrientation(SetAlignment=true, SetFacing=true)
   &ScaleColor(
    Scale Alpha=&FloatFromCurve(FloatCurve=DataInterface() {propertyValues {Curve {
        defaultValue=1
        keys [
         {arriveTangent=10.5294, interpMode=RCIM_Cubic, leaveTangent=10.5294, tangentMode=RCTM_User, time=0, value=0}
         {interpMode=RCIM_Cubic, tangentMode=RCTM_User, time=0.1, value=1}
        ]
       }}})
    ScaleA=true
    ScaleRGB=false
   )
  ]
  rendererClasses [&NiagaraSpriteRendererProperties]
  renderers [{rendererClass=&NiagaraSpriteRendererProperties}]
 }
}
summary {
 systemName=NS_S3800_Frontend
 emitters [
  {bEnabled=false, emitterName=Leaves, rendererClasses [&NiagaraSpriteRendererProperties]}
  {emitterName=Leaves001, rendererClasses [&NiagaraSpriteRendererProperties]}
 ]
 userVariables []
}
